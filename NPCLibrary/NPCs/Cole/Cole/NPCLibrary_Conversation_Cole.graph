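# Conversation graph for the NPC Cole, giving his rock-loving greeting and handing off to shared conversation branches.
# The Cole entry tag leads to one Speech node with a default line and prioritized contextual variants for time since last match, first-ever conversation and wearing the FutureSamurai outfit.
# Variant priorities come from Tandem.Speech.Priority curve rows.
# The speech then branches through link nodes to the shared NPCLibrary Quests and BasicServices entry tags.
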ConversationDatabase /NPCLibrary/NPCs/Cole/Cole/NPCLibrary_Conversation_Cole
entry_points [{tag=Conversation.Entry.Cole, destinations [#1]}]
nodes [
 ConversationEntryPointNode_7 {
  id=#1
  class=ConversationEntryPointNode
  outputs [#2]
  properties {entryTag {tagName=Conversation.Entry.Cole}}
 }
 Speech_C_4 {
  id=#2
  class=Speech_C
  outputs [#3, #4]
  properties {
   nodeName=""
   generalConfig {
    defaultMessage="Oh man, do you like rocks too?"
    contextualMessages [
     {
      message="Couldn't stay away from these rocks either, huh? So beautiful."
      contextRequirements [{requirement {class=FortControllerRequirement_CheckTimeSinceLastPlayedMatch}}]
      priority {curve {rowName=Tandem.Speech.Priority.WhileSincePlayed}}
     }
     {
      message="Lemme guess - You want to get in on the rocks too, huh? Don't blame ya!"
      contextRequirements [{requirement {class=FortControllerRequirement_IsFirstEverConversationWithNPC}}]
      priority {curve {rowName=Tandem.Speech.Priority.FirstEverInteraction}}
     }
     {
      message=I heard you sleep in a lamp. Is that true..?
      contextRequirements [{requirement {
         class=FortControllerRequirement_HasCID
         properties {allowedCIds [/BRCosmetics/Athena/Items/Cosmetics/Characters/CID_966_Athena_Commando_M_FutureSamurai]}
        }}]
      priority {curve {rowName=Tandem.Speech.Priority.OutfitReaction}}
     }
    ]
   }
  }
 }
 ConversationLinkNode_0 {id=#4, class=ConversationLinkNode, properties {remoteEntryTag {tagName=Conversation.Entry.BasicServices}}}
 ConversationLinkNode_1 {id=#3, class=ConversationLinkNode, properties {remoteEntryTag {tagName=Conversation.Entry.NPCLibrary.Quests}}}
]
